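# Animation Blueprint for the M_MED_Speed backpack cosmetic that adjusts the pack bone as the character's back bends.
# Each update maps spine_05 yaw relative to spine_01 on the owner mesh to additive pack roll and translation.
# The AnimGraph applies these, then a fixed offset and a skydiving-gated bone modifier.
# A smoothed back-arch alpha from the body mesh and a parachute-open blend function are present but unused by the pose.
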AnimBlueprint /BRCosmetics/Accessories/FORT_Backpacks/Backpack_M_MED_Speed/Meshes/M_MED_Speed_Pack_AnimBP
target_skeleton=&/M_MED_Speed_Pack_Skeleton
graphs [
 AnimGraph {nodes [
   n1 = "Transform (Modify) Bone - Bone: pack"(ComponentPose=Local To Component(LocalPose=Input Pose()), Translation=Get backpack_translation(), Rotation=Get backpack_rotate()) {
    rotationMode=BMM_Additive
    scaleMode=BMM_Additive
    scaleSpace=BCS_BoneSpace
    translationMode=BMM_Additive
    alphaBoolBlend {blendOption=Linear}
    boneToModify {boneName=pack}
   }
   n2 = "Transform (Modify) Bone - Bone: pack"(ComponentPose=@n1, Translation="0,-3,5", Rotation="0,0,0") {
    rotationMode=BMM_Additive
    rotationSpace=BCS_ParentBoneSpace
    translationMode=BMM_Additive
    translationSpace=BCS_ParentBoneSpace
    alphaBoolBlend {blendInTime=0.2, blendOption=Sinusoidal, blendOutTime=0.2}
    boneToModify {boneName=pack}
    translation {y=-3, z=5}
   }
   n3 = "Transform (Modify) Bone - Bone: pack"!(ComponentPose=@n2, bAlphaBoolEnabled:bool=Get bIsSkyDiving(), Translation="0,0,0", Rotation="0,0,0", Scale="0.9,0.9,0.9") -> (Pose) {
    alphaInputType=Bool
    rotationMode=BMM_Additive
    rotationSpace=BCS_ParentBoneSpace
    translationMode=BMM_Additive
    translationSpace=BCS_ParentBoneSpace
    alphaBoolBlend {blendInTime=0.2, blendOption=Sinusoidal, blendOutTime=0.2}
    boneToModify {boneName=pack}
    scale=<0.9 0.9 0.9>
   }
   Output Pose(Result=Component To Local(ComponentPose=@n3)) {layerGroup=None}
  ]}
 BlendOnParachute {nodes [
   BlendOnParachute() {is=title K2Node_FunctionEntry}
   n1 = float * float(A=Get Delta Time X(), B=0.1) {is=title K2Node_PromotableOperator}
   n2 = Get parachuteBlend() -> (parachuteBlend:double)
   > n3 = Set parachuteBlend(parachuteBlend:double=float + float(A=Get parachuteBlend(), B=Lerp(A=@n2, B=1, Alpha=@n1))) -> (Output_Get:double)
   > Branch(Condition="Nearly Equal (Float)"(A=@n3, B=1, ErrorTolerance=0.001))
   > Set parachuteOpened(parachuteOpened:bool=true) -> (Output_Get:bool)
  ]}
 EventGraph {nodes [
   n1 = Event BlueprintUpdateAnimation()
   > n2 = Sequence()
   > Set Delta Time X(Delta Time X=@n1.DeltaTimeX)
   > Is Valid(InputObject=Get Body() -> (Body:Skeletal Mesh Component Budgeted Object Reference))
   n3 = Get Body()
   n4 = BreakRotator(InRot=BreakTransform(InTransform=GetDeltaTransformFromRefPose(self=@n3, BoneName=spine_05, BaseName=pelvis)).Rotation)
   > Set BackArchAlpha(BackArchAlpha:double=FInterpTo(
     Current=Get BackArchAlpha() -> (BackArchAlpha:double)
     Target=MapRangeClamped(Value=@n4.Yaw, InRangeA=0, InRangeB=-30)
     DeltaTime=Get Delta Time X()
     InterpSpeed=1.5
    )) -> (Output_Get:double)
   n5 = GetOwnerSkeletalMeshComponent()
   n6 = BreakRotator(InRot=BreakTransform(InTransform=GetDeltaTransformFromRefPose(self=@n5, BoneName=spine_05, BaseName=spine_01)).Rotation)
   Set backpack_rotate(execute=@n2.then_1, backpack_rotate=MakeRotator(Roll=MapRangeClamped(Value=@n6.Yaw, InRangeA=0, InRangeB=-30, OutRangeA=3, OutRangeB=25)))
   > Set backpack_translation(backpack_translation=MakeVector(Y=MapRangeClamped(Value=@n6.Yaw, InRangeA=0, InRangeB=-30, OutRangeB=-8)))
   n7 = Event OnCharacterCustomizationLoaded()
   > Set Backpack(Backpack:Skeletal Mesh Component Object Reference=@n7.CustomBaseMesh) -> (Output_Get:Skeletal Mesh Component Object Reference)
   > Set Body(Body:Skeletal Mesh Component Budgeted Object Reference=Get Array Item with EFort Custom Part Type(Array=@n7.CustomMeshParts)) -> (Output_Get:Skeletal Mesh Component Budgeted Object Reference)
   TryGetPawnOwner()
  ]}
]
variables [
 AimAtTarget {type=Vector}
 BackArchAlpha {type=number}
 Backpack {type=SkeletalMeshComponent}
 backpack_rotate {type=Rotator}
 backpack_translation {type=Vector}
 Body {type=SkeletalMeshComponentBudgeted}
 bSkydiveDiveCorrection {type=boolean}
 bUseAimLocation {type=boolean}
 Delta Time X {type=number}
 parachuteBlend {type=number}
 parachuteClosedTrigger {type=boolean}
 parachuteOpened {type=boolean}
 parachuteOpenedTrigger {type=boolean}
]
